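; Masamune stand loot container that fades its ambient effects when opened and removes its sword mesh when openable cosmetics hide.
; BeginPlay binds OnInteract manually because the Blueprint interact event does not fire for BuildingContainers.
; Interaction sets FXFade, whose OnRep starts a timeline fading wind, glow, sparkle and Niagara leaf parameters, then destroys the second wind trail.
; Overrides always allow interaction and supply custom prompt text.
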(event EventBeginPlay
  (|CallParentFunction)
  ; Ideally this would be the BP OnInteract event but it's not fired for BuildingContainers and we're too close to HL to comfortably change at this point
  (Interaction|BindEventtoOnInteract (EventDispatchers|CreateEvent :!event Custom|OnInteract_Event)))

(event Custom|OnInteract_Event
  (params (SelfActor :type Building%20Actor%20Object%20Reference) (InteractingPawn :type Fort%20Pawn%20Object%20Reference))
  ; FX Fade When opened
  (Variables|Default|SetFXFade true))

(event Custom|CE_FXFade
  ; FX Fade When opened
  (bind (_direction _windfxfade _sparklessprite_size _glowfxfade01 _centerfx _spawnnumber) (Timeline|Timeline PlayFromStart)
    (:Update
      (Rendering|Material|SetScalarParameterValueonMaterials (Variables|Default|GetSM_MasamuneStand_WindTrail01) "AlphaIntensity" _windfxfade)
      (Rendering|Material|SetScalarParameterValueonMaterials (Variables|Default|GetSM_MasamuneStand_Particles) "Sprite Size" _sparklessprite_size)
      (Rendering|Material|SetScalarParameterValueonMaterials (Variables|Default|GetSM_MasamuneStand_SplineGlow) "AlphaIntensity" _glowfxfade01)
      (Rendering|Material|SetScalarParameterValueonMaterials (Variables|Default|GetSM_MasamuneStand_CenterGlow) "AlphaIntensity" _centerfx)
      (Effects|Components|ParticleSystem|SetFloatParameter (Variables|Default|GetNS_Masamune_AirflowLeaf) "SpawnNumber" _spawnnumber))
    (:Finished
      (Components|DestroyComponent (Variables|Default|GetSM_MasamuneStand_WindTrail02)))))

(event EventHideChestOpenableCosmetics
  (|CallParentFunction)
  (Components|DestroyComponent (Variables|Default|GetSwordMesh)))

(fn BlueprintCanInteract
  (params (InteractingPawn :type Fort%20Pawn%20Object%20Reference) (InteractionBeingAttempted :type EInteractionBeingAttempted%20Enum :default "FirstInteraction") (InteractionType :type TInteractionType%20Enum :default "IT_NoInteraction"))
  (return true))

(fn BlueprintGetInteractionString
  (params (InteractingPawn :type Fort%20Pawn%20Object%20Reference) (InteractionBeingAttempted :type EInteractionBeingAttempted%20Enum :default "FirstInteraction"))
  (return (Variables|Default|GetInteractText)))

(fn GetMinDistanceFromInteraction
  (return 0.0))

(fn OnRep_FXFade
  (if (Variables|Default|GetFXFade)
    (CallFunction|CEFXFade)))

(fn ConstructionScript)
